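# Physics asset for the TriStar two-handed pickaxe cosmetic, giving its dangling bandage bones bodies and joints for physical motion.
# The handle body combines a long shaft capsule, a sphere and two angled capsules near the head.
# Three bandage chains (a with four bones, b and c with two each) use zero-length capsules and hang off the handle.
# Every joint locks twist, and chain a has wider swing angles than b and c.
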PhysicsAsset /BRCosmetics/Weapons/FORT_Melee/Pickaxe_TriStar/Meshes/Pickaxe_Tri_Star_Physics
bodies {
 Melee_TwoHanded_Handle [
  {shapeType=Sphere, radius=8.6708, center=<1.0203 -0.2942 72.8353>}
  {shapeType=Capsule, length=88.8017, radius=1.8675, center=<1.0075 -0.4712 19.2765>}
  {shapeType=Capsule, length=25.5608, radius=4.4392, center=<0.8705 16.611 58.718>, rotation=<rot 0 0 -45.0009>}
  {shapeType=Capsule, length=20.6915, radius=4.4392, center=<0.6301 -16.7918 62.8823>, rotation=<rot 0 0 54.9967>}
 ]
 dyn_bandage_a_01 [{shapeType=Capsule, length=0, radius=1.1463, center=<1 0 0>}]
 dyn_bandage_a_02 [{shapeType=Capsule, length=0, radius=1.1463, center=<1 0 0>}]
 dyn_bandage_a_03 [{shapeType=Capsule, length=0, radius=1.1463, center=<1 0 0>}]
 dyn_bandage_a_04 [{shapeType=Capsule, length=0, radius=3.4276, center=<1 0 0>}]
 dyn_bandage_b_01 [{shapeType=Capsule, length=0, radius=1, center=<1 0 0>}]
 dyn_bandage_b_02 [{shapeType=Capsule, length=0, radius=1, center=<1 0 0>}]
 dyn_bandage_c_01 [{shapeType=Capsule, length=0, radius=1, center=<1 0 0>}]
 dyn_bandage_c_02 [{shapeType=Capsule, length=0, radius=1, center=<1 0 0>}]
}
constraints [
 {
  bone1Name=dyn_bandage_a_01
  bone2Name=Melee_TwoHanded_Handle
  swing1LimitDegrees=90
  swing2LimitDegrees=90
  twistLimitDegrees=45
  twistMotion=Locked
 }
 {
  bone1Name=dyn_bandage_a_02
  bone2Name=dyn_bandage_a_01
  swing1LimitDegrees=45
  swing2LimitDegrees=45
  twistLimitDegrees=45
  twistMotion=Locked
 }
 {
  bone1Name=dyn_bandage_a_03
  bone2Name=dyn_bandage_a_02
  swing1LimitDegrees=45
  swing2LimitDegrees=45
  twistLimitDegrees=45
  twistMotion=Locked
 }
 {
  bone1Name=dyn_bandage_a_04
  bone2Name=dyn_bandage_a_03
  swing1LimitDegrees=45
  swing2LimitDegrees=45
  twistLimitDegrees=45
  twistMotion=Locked
 }
 {
  bone1Name=dyn_bandage_b_01
  bone2Name=Melee_TwoHanded_Handle
  swing1LimitDegrees=30
  swing2LimitDegrees=30
  twistLimitDegrees=45
  twistMotion=Locked
 }
 {
  bone1Name=dyn_bandage_b_02
  bone2Name=dyn_bandage_b_01
  swing1LimitDegrees=20
  swing2LimitDegrees=20
  twistLimitDegrees=45
  twistMotion=Locked
 }
 {
  bone1Name=dyn_bandage_c_01
  bone2Name=Melee_TwoHanded_Handle
  swing1LimitDegrees=30
  swing2LimitDegrees=30
  twistLimitDegrees=45
  twistMotion=Locked
 }
 {
  bone1Name=dyn_bandage_c_02
  bone2Name=dyn_bandage_c_01
  swing1LimitDegrees=20
  swing2LimitDegrees=20
  twistLimitDegrees=45
  twistMotion=Locked
 }
]
